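# Niagara system for the Fusion vehicle boost cosmetic, trailing red-tinted ribbons and sprites from skeletal mesh boost ports.
# The ribbon emitter drives three ribbon renderers with a derived glow width, while the sprite emitter partitions rotating, growing, fading particles across port sockets.
# Both inherit decaying source movement, switch color for frontend preview, and vary values by Coro, FrontEnd or FN through mask modules.
# Painted-variant override modules are disabled in both emitters.
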NiagaraSystem /VehicleCosmetics/Boosts/Boost_Fusion/FX/NS_Boost_Fusion
dependencies {
 usedDataInterfaces [&NiagaraDataInterfaceCurve, &NiagaraDataInterfaceSkeletalMesh, &NiagaraDataInterfaceArrayFloat3]
 usedDynamicInputs [&SelectVectorFromArray]
 usedModules [
  &self:SystemSpawnScript.SystemScriptSource.SystemScriptGraph.NiagaraNodeAssignment_0.SetVariables_#1
  &self:SystemSpawnScript.SystemScriptSource.SystemScriptGraph.NiagaraNodeAssignment_1.SetVariables_#2
  &self:SystemSpawnScript.SystemScriptSource.SystemScriptGraph.NiagaraNodeAssignment_2.SetVariables_#3
  &SystemState
 ]
}
emitters {
 NE_VC_Boost_Ribbon_Parent {
  emitterUpdateScript [
   &EmitterState()
   &VC_FilteredSocketCount()
   &VC_RateQualityHandler(Default Rate:NiagaraFloat=&Multiply_Float(
     A=95
     B=/VehicleCosmetics/FX/Niagara/Modules/VC_Mask_Floats(Coro Value:NiagaraFloat=1, FrontEnd Value:NiagaraFloat=0.75, FN Value:NiagaraFloat=0.75)
    ))
   &VC_SpawnRibbon_v1(Number of Ribbons=$Emitter.PortCount, SpawnRate=$Emitter.Spawn Rate)
  ]
  particleSpawnScript [
   &VC_PaintedVariantsOverride() {enabled=false}
   &VC_InitializeRibbon_v1(
    Color=&MakeCustomLinearColorFromBool(Bool=$User.bIsFrontendPreview, True Color=<rgba 1 0 0.0174 1>, False Color=<rgba 0.375 0.0586 0.0701 1>)
    Enable Skeletal Position=true
    Lifetime=1
    Ribbon Width=50
   )
   &AddVelocity(Velocity=&VC_Mask_Vectors(Coro Value:Vector3f=<-400 0 0>, FrontEnd Value:Vector3f=<-2000 0 0>, FN Value:Vector3f=<-400 0 0>)) {module=AddVelocity001}
   /VehicleCosmetics/FX/VC_FXParents/Modules/DLR_ProductPreviews(GameModeID:NiagaraInt32=$System.GameModeID, bPreviewVehicleSpeed:NiagaraBool=$System.bPreviewVehicleSpeed)
  ]
  particleUpdateScript [
   &ParticleState()
   &InheritSourceMovement(
    Applied Position Scale=&VectorFromFloat(Value=&FloatFromCurve(FloatCurve=DataInterface() {propertyValues {Curve {
         defaultValue=1
         keys [
          {arriveTangent=-1.7885, interpMode=RCIM_Cubic, leaveTangent=-1.7885, tangentMode=RCTM_User, time=0.05, value=1}
          {interpMode=RCIM_Cubic, tangentMode=RCTM_User, time=1, value=0}
         ]
        }}}))
    ApplyForce=false
    ApplyPosition=true
    ApplyPositionFalloff=false
   )
   &SolveForcesAndVelocity(Clamp Velocity=false, Limit Acceleration=false)
   &VC_UpdateRibbon_v1(Enable Skeletal Position=true)
   &ScaleRibbonWidth(Ribbon Width Scale=&FloatFromCurve(FloatCurve=DataInterface() {propertyValues {Curve {keys [
         {arriveTangent=0.0791, interpMode=RCIM_Cubic, leaveTangent=0.0791, tangentMode=RCTM_User, time=0, value=1}
         {arriveTangent=-1.0476, interpMode=RCIM_Cubic, leaveTangent=-1.0476, tangentMode=RCTM_User, time=0.6538, value=0.8439}
         {arriveTangent=-3.282, interpMode=RCIM_Cubic, leaveTangent=-3.282, tangentMode=RCTM_User, time=1, value=0}
        ]}}}))
   /SetVariables(Particles.GlowRibbonWidth:NiagaraFloat=&Multiply_Float(A=$Particles.RibbonWidth, B=2.2)) {isSetParameters=true}
   &ScaleColor(
    Scale RGB=&VectorFromFloat(Value=&FloatFromCurve(
      FloatCurve=DataInterface() {propertyValues {Curve {
         defaultValue=1
         keys [
          {arriveTangent=-5.2861, interpMode=RCIM_Cubic, leaveTangent=-5.2861, tangentMode=RCTM_User, time=0, value=2}
          {interpMode=RCIM_Cubic, tangentMode=RCTM_User, time=0.65, value=0.4}
         ]
        }}}
      Scale Curve=&VC_Mask_Floats(Coro Value:NiagaraFloat=10, FrontEnd Value:NiagaraFloat=5, FN Value:NiagaraFloat=0.5)
     ))
    ScaleA=true
    ScaleRGB=true
   )
  ]
  rendererClasses [&NiagaraRibbonRendererProperties]
  renderers [
   {rendererClass=&NiagaraRibbonRendererProperties}
   {rendererClass=&NiagaraRibbonRendererProperties, rendererIndex=1}
   {rendererClass=&NiagaraRibbonRendererProperties, rendererIndex=2}
  ]
 }
 NE_VC_Boost_Sprite_Parent001 {
  emitterUpdateScript [
   &EmitterState()
   &VC_FilteredSocketCount()
   &SpawnRate(
    SpawnRate=&Multiply_Float(
     A=90
     B=&VC_Mask_Floats(Coro Value:NiagaraFloat=1.2, FrontEnd Value:NiagaraFloat=0.5, FN Value:NiagaraFloat=0.5)
    )
    Use Spawn Probability=false
   )
  ]
  particleSpawnScript [
   &VC_PaintedVariantsOverride() {enabled=false}
   &InitializeParticle(
    Color=&MakeCustomLinearColorFromBool(Bool=$User.bIsFrontendPreview, True Color=<rgba 2 0.02 0 1>, False Color=<rgba 2 0.0938 0.1414 1>)
    Lifetime=&RandomRangeFloat(
     Minimum=&VC_Mask_Floats(Coro Value:NiagaraFloat=0.8, FrontEnd Value:NiagaraFloat=0.7, FN Value:NiagaraFloat=0.5)
     Maximum=&VC_Mask_Floats(Coro Value:NiagaraFloat=0.8, FrontEnd Value:NiagaraFloat=1, FN Value:NiagaraFloat=0.5)
    )
    Sprite Rotation Mode=::NewEnumerator2
    Sprite Size Mode=::NewEnumerator1
    Sprite UV Mode=::NewEnumerator3
    Uniform Sprite Size Max=125
    Uniform Sprite Size Min=90
    UsePositionOffset=false
    Write Material Random=true
   )
   &PartitionParticles(Number Of Partitions=$Emitter.PortCount)
   &SkeletalMeshLocation(
    Bone / Socket Index=$Particles.PartitionParticles.Partition
    Bone / Socket Sampling Mode=::NewEnumerator5
    MaskSpawnGroup=false
    Mesh Sampling Type=::NewEnumerator6
    Skeletal Mesh=$System.SkeletalMesh
   )
   &AddVelocity(Velocity=&VC_Mask_Vectors(Coro Value:Vector3f=<-400 0 0>, FrontEnd Value:Vector3f=<-1200 0 0>, FN Value:Vector3f=<-1200 0 0>))
   &DLR_ProductPreviews(GameModeID:NiagaraInt32=$System.GameModeID, bPreviewVehicleSpeed:NiagaraBool=$System.bPreviewVehicleSpeed)
  ]
  particleUpdateScript [
   &ParticleState()
   &SpriteRotationRate(Rotation Rate=&Multiply_Float(
     A=&RandomRangeFloat(Minimum=20, Maximum=60)
     B=&UniformAOrBFloat(A=-1, Evaluation Type=::NewEnumerator0, Recalculate Random Each Loop=true, Override Seed=false)
    ))
   &ScaleSpriteSize(
    Scale Sprite Size Mode=::NewEnumerator0
    Uniform Scale Factor=&VC_Mask_Floats(
     Coro Value:NiagaraFloat=&FloatFromCurve(FloatCurve=DataInterface() {propertyValues {Curve {keys [
          {arriveTangent=4.7523, interpMode=RCIM_Cubic, leaveTangent=4.7523, leaveTangentWeight=0.4027, tangentMode=RCTM_User, time=0, value=0.65}
          {arriveTangent=2.6951, interpMode=RCIM_Cubic, leaveTangent=2.6951, tangentMode=RCTM_User, time=0.4023, value=1.7605}
          {arriveTangent=1.028, interpMode=RCIM_Cubic, leaveTangent=1.028, tangentMode=RCTM_User, time=0.9854, value=2.6}
         ]}}})
     FrontEnd Value:NiagaraFloat=&FloatFromCurve(FloatCurve=DataInterface() {propertyValues {Curve {keys [
          {arriveTangent=4.7523, interpMode=RCIM_Cubic, leaveTangent=4.7523, leaveTangentWeight=0.4027, tangentMode=RCTM_User, time=0, value=0.65}
          {arriveTangent=3.9212, interpMode=RCIM_Cubic, leaveTangent=3.9212, tangentMode=RCTM_User, time=0.3969, value=1.506}
          {arriveTangent=3.5867, interpMode=RCIM_Cubic, leaveTangent=3.5867, tangentMode=RCTM_User, time=0.9854, value=2.4525}
         ]}}})
     FN Value:NiagaraFloat=&FloatFromCurve(FloatCurve=DataInterface() {propertyValues {Curve {keys [
          {arriveTangent=4.7523, interpMode=RCIM_Cubic, leaveTangent=4.7523, leaveTangentWeight=0.4027, tangentMode=RCTM_User, time=0, value=0.65}
          {arriveTangent=3.9212, interpMode=RCIM_Cubic, leaveTangent=3.9212, tangentMode=RCTM_User, time=0.3969, value=1.506}
          {arriveTangent=3.5867, interpMode=RCIM_Cubic, leaveTangent=3.5867, tangentMode=RCTM_User, time=0.9854, value=2.4525}
         ]}}})
    )
   )
   &InheritSourceMovement(
    Applied Position Scale=&Multiply_VectorByFloat(
     Vector=<1 1 1>
     Float=&FloatFromCurve(FloatCurve=DataInterface() {propertyValues {Curve {
         defaultValue=1
         keys [
          {arriveTangent=-3.3465, interpMode=RCIM_Cubic, leaveTangent=-3.3465, tangentMode=RCTM_User, time=0, value=1}
          {tangentMode=RCTM_User, time=1, value=0}
         ]
        }}})
    )
    ApplyForce=false
    ApplyPosition=true
    ApplyPositionFalloff=false
   )
   &SolveForcesAndVelocity(Clamp Velocity=false, Limit Acceleration=false)
   &ScaleColor(
    Scale Alpha=&FloatFromCurve(
     FloatCurve=DataInterface() {propertyValues {Curve {keys [
         <at -0.0013 1.0364>
         {
          arriveTangentWeight=0.1177
          interpMode=RCIM_Cubic
          leaveTangent=-1.9306
          leaveTangentWeight=0.2059
          tangentMode=RCTM_Break
          time=0.3
          value=0.9765
         }
         {arriveTangent=-1.4103, interpMode=RCIM_Cubic, leaveTangent=-1.4103, tangentMode=RCTM_User, time=1, value=0}
        ]}}}
     Scale Curve=&RandomRangeFloat(
      Minimum=0.1
      Maximum=&VC_Mask_Floats(Coro Value:NiagaraFloat=0.4, FrontEnd Value:NiagaraFloat=0.6, FN Value:NiagaraFloat=0.5)
     )
    )
    Scale RGB=&VectorFromFloat(Value=&FloatFromCurve(
      FloatCurve=DataInterface() {propertyValues {Curve {keys [
          {arriveTangent=-38.1004, interpMode=RCIM_Cubic, leaveTangent=-38.1004, leaveTangentWeight=0.3333, tangentMode=RCTM_User, time=0.04, value=8}
          {arriveTangent=-7.5756, arriveTangentWeight=0.3333, interpMode=RCIM_Cubic, leaveTangent=-7.5756, tangentMode=RCTM_User, time=0.5, value=1}
         ]}}}
      Scale Curve=&VC_Mask_Floats(Coro Value:NiagaraFloat=0.7, FrontEnd Value:NiagaraFloat=1, FN Value:NiagaraFloat=0.5)
     ))
    ScaleA=true
    ScaleRGB=true
   )
   &DynamicMaterialParameters(
    Index 1 Param 1=&FloatFromCurve(FloatCurve=DataInterface() {propertyValues {Curve {keys [<at 0 5>,<at 0.2 1>]}}})
    Index 1 Param 2=&RandomRangeFloat(
     Minimum=0.564
     Maximum=&VC_Mask_Floats(Coro Value:NiagaraFloat=0.9, FrontEnd Value:NiagaraFloat=1.1, FN Value:NiagaraFloat=0.5)
    )
    Index 2 Param 1=&RandomRangeFloat(Minimum=0.12, Maximum=0.25)
    Index 2 Param 2=&RandomRangeFloat(Minimum=0, Maximum=0.75)
    Index 2 Param 3=&FloatFromCurve(FloatCurve=DataInterface() {propertyValues {Curve {keys [<at -0.0025 0.7>,<at 0.4976 0.4009>,<at 1 0>]}}})
    Module.Index0Float0Write=true
    Module.Index0Float1Write=true
    Module.Index0Float2Write=true
    Module.Index0Float3Write=true
    Module.Index1Float0Write=true
    Module.Index1Float1Write=true
    Module.Index1Float2Write=true
    Module.Index1Float3Write=true
    Module.Index2Float0Write=true
    Module.Index2Float1Write=true
    Module.Index2Float2Write=true
    Module.Index2Float3Write=true
    Write Parameter Index 1=::NewEnumerator1
    Write Parameter Index 2=::NewEnumerator1
   )
  ]
  rendererClasses [&NiagaraSpriteRendererProperties]
  renderers [{rendererClass=&NiagaraSpriteRendererProperties}]
 }
}
summary {
 systemName=NS_Boost_Fusion
 emitters [
  {emitterName=NE_VC_Boost_Ribbon_Parent, rendererClasses [&NiagaraRibbonRendererProperties]}
  {emitterName=NE_VC_Boost_Sprite_Parent001, rendererClasses [&NiagaraSpriteRendererProperties]}
 ]
 userVariables [
  User.bIsFrontendPreview {type=NiagaraBool}
  User.bIsPainted {type=NiagaraBool}
  User.blsIsCoroColor {type=NiagaraBool}
  User.bPreviewVehicleSpeed {type=NiagaraBool}
  User.ColorVariantID {type=NiagaraInt32}
  User.GameModeID {type=NiagaraInt32}
  User.PaintedColor {type=LinearColor, description="", defaultValue {struct=LinearColor, value=<rgba 1 1 1 1>}}
  User.PaintedScalar {type=NiagaraFloat, defaultValue=1, description=""}
  User.PortCount {type=NiagaraInt32, defaultValue=2}
 ]
}
